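# Unlit additive, two-sided material for lightning spline effects on the OvenDrastic pickaxe.
# A lightning texture pans in quantized time steps, then thresholds into core and glow masks.
# These masks drive a glow-to-core color blend, tinted by time of day, and opacity with view-angle and depth fades.
# SplineThicken sets width from an end-to-middle sine profile, jittered by layered time sines and scrolling noise.
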Material /BRCosmetics/Weapons/FORT_Melee/Pickaxe_OvenDrastic/FX/M_Pickaxe_OvenDrastic_LightningSpline
graph {
 properties {
  bAutomaticallySetUsageInEditor=false
  bUsedWithMeshDeformer=true
  bUsedWithSkeletalMesh=true
  blendMode=BLEND_Additive
  shadingModel=MSM_Unlit
  twoSided=true
  usedShadingModels=Unlit
 }
 nodes [
  n1 = MaterialFunctionCall(XY=TextureCoordinate()) {materialFunction=&Swizzle}
  n2 = AppendVector(A=Constant() {r=1}, B=ScalarParameter() {defaultValue=1, parameterName=Tile, sliderMax=5})
  n3 = ScalarParameter() {defaultValue=5, parameterName=Steps, sliderMax=20}
  n4 = Floor(None=Multiply(A=Time(), B=@n3) {constB=2})
  n5 = Constant()
  n6 = ScalarParameter() {defaultValue=1, parameterName=Speed, sliderMax=5, sliderMin=-5}
  n7 = TextureSample(UVs=Panner(Coordinate=Multiply(A=@n1.YX, B=@n2), Time=Divide(A=@n4, B=@n3), Speed=AppendVector(A=@n5, B=@n6))) {samplerType=SAMPLERTYPE_LinearGrayscale, texture=&/T_Pickaxe_OvenDrastic_Lightning}
  n8 = ScalarParameter() {defaultValue=1, group=Shape, parameterName=Shape Core Size, sliderMax=10}
  n9 = Subtract(A=@n7.R, B=OneMinus(None=Multiply(A=@n8)))
  n10 = Add(A=ScalarParameter() {defaultValue=1, group=Shape, parameterName=Shape Core Softness, sliderMax=10}) {constB=0.0001}
  n11 = ScalarParameter() {defaultValue=1, group=Shape, parameterName=Shape Glow Size, sliderMax=5}
  n12 = Subtract(
   A=Power(Base=@n7.R, Exp=ScalarParameter() {defaultValue=1, group=Shape, parameterName=Shape Glow Power, sliderMax=2})
   B=OneMinus(None=Multiply(A=@n11))
  )
  n13 = Add(A=ScalarParameter() {defaultValue=1, group=Shape, parameterName=Shape Glow Softness, sliderMax=10}) {constB=0.0001}
  n14 = Add(A=Saturate(None=Divide(A=@n9, B=@n10)), B=Saturate(None=Divide(A=@n12, B=@n13)))
  n15 = ScalarParameter() {defaultValue=1, parameterName=FadeSensitivity, sliderMax=10}
  DepthFade(
   Opacity=Multiply(
    A=@n14
    B=MaterialFunctionCall("FadeSensitivity (optional)"=@n15) {materialFunction=&FadeBasedOnViewAngle}.Result
   )
   FadeDistance=ScalarParameter() {defaultValue=5, parameterName=Depth Fade, sliderMax=100}
  )
  n16 = Sine(None=ComponentMask(None=TextureCoordinate()) {r=true}) {period=2}
  n17 = Time()
  n18 = Sine(None=@n17) {period=0.37}
  n19 = Add(A=ConstantBiasScale(None=Sine(None=@n17) {period=0.69}), B=Multiply(A=ConstantBiasScale(None=@n18)) {constB=0.76})
  n20 = Panner(Coordinate=Multiply(A=TextureCoordinate()) {constB=0.75}) {speedX=0.5}
  n21 = Multiply(
   A=TextureSample(UVs=ComponentMask(None=@n20) {r=true}) {samplerType=SAMPLERTYPE_Grayscale, texture=&LowResBlurredNoise}.R
   B=ScalarParameter() {parameterName=Width Jitter Intensity}
  )
  n22 = Multiply(
   A=LinearInterpolate(
    A=ScalarParameter() {defaultValue=10, parameterName=Width End, sliderMax=50}
    B=ScalarParameter() {defaultValue=10, parameterName=Width Middle, sliderMax=50}
    Alpha=Power(Base=@n16)
   )
   B=LinearInterpolate(
    A=ScalarParameter() {defaultValue=0.8, parameterName=Width 1, sliderMax=3}
    B=ScalarParameter() {defaultValue=1.2, parameterName=Width 2, sliderMax=3}
    Alpha=Multiply(A=@n19, B=@n21)
   ) {constA=0.8}
  )
  MaterialFunctionCall(WidthBase=@n22, WidthTip=@n22, Expand U or V UV Channel=StaticBoolParameter() {parameterName="Expand U or V UV Channel (SB)"}) {materialFunction=&SplineThicken}
  n23 = VectorParameter() {parameterName=Color_Morning, defaultValue=<rgba 1 1 1 1>}
  n24 = VectorParameter() {parameterName=Color_Day, defaultValue=<rgba 1 1 1 1>}
  n25 = VectorParameter() {parameterName=Color_Evening, defaultValue=<rgba 1 1 1 1>}
  n26 = VectorParameter() {parameterName=Color_Night, defaultValue=<rgba 1 1 1 1>}
  n27 = VectorParameter() {parameterName=Color Glow, defaultValue=<rgba 0 0.3832 1 1>}
  n28 = VectorParameter() {parameterName=Color Core, defaultValue=<rgba 0.3385 0.592 1 1>}
  MaterialFunctionCall(Emissive=Multiply(
    A=MaterialFunctionCall(Color_Morning=@n23.RGB, Color_Day=@n24.RGB, Color_Evening=@n25.RGB, Color_Night=@n26.RGB) {materialFunction=&Color_Time_of_Day}.Result
    B=LinearInterpolate(A=@n27.RGB, B=@n28.RGB, Alpha=@n14)
   )) {materialFunction=&MF_FX_EmissiveAdjustments}
 ]
}
